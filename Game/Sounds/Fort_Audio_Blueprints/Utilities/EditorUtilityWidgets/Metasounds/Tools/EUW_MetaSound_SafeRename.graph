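# Editor utility widget that preserves input defaults in MetaSound Sources referencing a selected MetaSound while one of its inputs is renamed.
# Cache Values finds referencers through the Asset Registry and stores each one's old-name input default in a map, using MetaSound builders.
# After the user renames the input manually, Safe Rename writes cached literals to the new name and saves each asset.
# The cache clears on construct, destruct and result confirmation.
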EditorUtilityWidgetBlueprint /Game/Sounds/Fort_Audio_Blueprints/Utilities/EditorUtilityWidgets/Metasounds/Tools/EUW_MetaSound_SafeRename
graphs [
 ClearCachedArrays {nodes [
   ClearCachedArrays() {is=title K2Node_FunctionEntry}
   n1 = Get TMapReferences() -> (TMapReferences:Map of Meta Sound Source Object References to Meta Sound Literal Structures)
   > n2 = Branch(Condition="integer > integer"(A=Length!(TargetMap:Map of Meta Sound Source Object References to Meta Sound Literal Structures=@n1) -> (ReturnValue:int=0) {is=title K2Node_CallFunction}))
   > n3 = Clear!(execute, TargetMap:Map of Meta Sound Source Object References to Meta Sound Literal Structures=@n1) -> (then) {is=title K2Node_CallFunction}
   n4 = Get ReferencedMS() -> (ReferencedMS:Array of Meta Sound Source Object References)
   n5 = Branch(
    execute=[@n3.then, @n2.else]
    Condition="integer > integer"(A=Length(TargetArray:Array of Meta Sound Source Object References=@n4) {is=title K2Node_CallArrayFunction})
   )
   Return Node(execute=[@n5.else, @n5.then])
  ]}
 EventGraph {nodes [
   "On Pressed (CacheValues)"() {is=title K2Node_ComponentBoundEvent}
   > ClearCachedArrays() {is=title K2Node_CallFunction}
   > n1 = GetSelectedAssetsOfClass(AssetClass=/Script/Engine.SoundWave)
   > Branch(Condition="integer > integer"(A=Length(TargetArray:Array of Object References=@n1)))
   > n2 = For Each Loop(Array:Array of Object References=@n1) -> (Array Element:Object Reference)
   > n3 = Cast To MetaSoundSource!(execute, Object=@n2) -> (then, CastFailed, AsMetaSound Source:Meta Sound Source Object Reference) {is=title K2Node_DynamicCast}
   > Set SelectedMSName(SelectedMSName:String=GetDisplayName(Object=@n3)) -> (Output_Get:String)
   > n4 = Find Referencers!(execute:Exec, Object:Object Reference=@n3) -> (Iteration:Exec, Referencer:Object Reference, Completed:Exec) {is=title K2Node_MacroInstance}
   n5 = Get ReferencedMS()
   n6 = Length(TargetArray:Array of Meta Sound Source Object References=@n5)
   Branch(execute=@n4.Completed, Condition="integer > integer"(A=@n6))
   > n7 = For Each Loop(Array:Array of Meta Sound Source Object References=@n5) -> (Array Element:Meta Sound Source Object Reference)
   n8 = "To String (Text)"(InText=Get Text(self=Get OldName() -> (OldName:Editor Utility Editable Text Box Object Reference)))
   n9 = Get SelectedMSName() -> (SelectedMSName:String)
   n10 = Length(TargetMap=Get TMapReferences()) -> (ReturnValue=0) {is=title K2Node_CallFunction}
   Show Message Dialog(
    execute=@n7.Completed
    Title=MakeLiteralText(Value="LOCTEXT(\"Cached Values Result\")")
    Message=Select(
     Option 0:Text=Format Text(Format="LOCTEXT(\"Failed to retrieve graph input values!!\")")
     Option 1:Text=Format Text!(
      Format="LOCTEXT(\"You\\'ve succesfully cached {Int} input values for {MSName} {MSInput} input. You can now rename {OldInput} to its new value and then press \\'Safe Rename\\'.\")"
      Int:int=@n6
      MSName:String=@n9
      MSInput:String=@n8
      OldInput:String=@n8
     ) -> (Result)
     Index="integer > integer"(A=@n10)
    ) -> (ReturnValue:Text)
    MessageType=Ok
    DefaultValue=Ok
    MessageCategory=Info
   )
   n11 = Find Or Begin Building MetaSound Asset!(
    self:Meta Sound Editor Subsystem Object Reference=Get MetaSoundEditorSubsystem() -> (ReturnValue:Meta Sound Editor Subsystem Object Reference) {is=title K2Node_GetEditorSubsystem}
    MetaSound:Meta Sound Document Interface Interface=@n7
   ) -> (ReturnValue:Meta Sound Builder Base Object Reference)
   n12 = Get OldName()
   n13 = String To Name(InString="To String (Text)"(InText=Get Text(self=@n12)))
   n14 = FindGraphInputNode!(execute=@n7.LoopBody, self:Meta Sound Builder Base Object Reference=@n11, InputName:Name=@n13) -> (
    Succeeded:Exec
    Failed:Exec
    OutDataType:Name=None
    NodeOutputHandle:Meta Sound Node Output Handle Structure
    ReturnValue:Meta Sound Node Handle Structure
   )
   > n15 = FindNodeInputByName!(
    execute
    self:Meta Sound Builder Base Object Reference=@n11
    NodeHandle:"Meta Sound Node Handle Structure (by ref)"=@n14.ReturnValue
    InputName:Name=@n13
   ) -> (Succeeded:Exec, Failed:Exec, ReturnValue:Meta Sound Node Input Handle Structure)
   > n16 = GetNodeInputDefault!(execute, self:Meta Sound Builder Base Object Reference=@n11, InputHandle:"Meta Sound Node Input Handle Structure (by ref)"=@n15) -> (Succeeded:Exec, Failed:Exec, ReturnValue:Meta Sound Literal Structure)
   > Add!(
    execute
    TargetMap:Map of Meta Sound Source Object References to Meta Sound Literal Structures=Get TMapReferences()
    Key:"Meta Sound Source Object Reference (by ref)"=@n7
    Value:"Meta Sound Literal Structure (by ref)"=@n16
   ) -> (then) {is=title K2Node_CallFunction}
   n17 = Cast To MetaSoundSource(execute=@n4.Iteration, Object=@n4)
   > Add Unique(TargetArray:Array of Meta Sound Source Object References=Get ReferencedMS(), NewItem:"Meta Sound Source Object Reference (by ref)"=@n17)
   "On Pressed (SafeRename)"() {is=title K2Node_ComponentBoundEvent}
   n18 = Get TMapReferences()
   > Branch(Condition="integer > integer"(A=Length(TargetMap=@n18) -> (ReturnValue=0)))
   > n19 = Keys!(execute, TargetMap:Map of Meta Sound Source Object References to Meta Sound Literal Structures=@n18) -> (then, Keys:Array of Meta Sound Source Object References)
   > n20 = For Each Loop(Array=@n19)
   n21 = "To String (Text)"(InText=Get Text(self=Get NewName() -> (NewName:Editor Utility Editable Text Box Object Reference)))
   n22 = Show Message Dialog(
    execute=@n20.Completed
    Title=MakeLiteralText(Value="LOCTEXT(\"Safe Rename Result\")")
    Message=Format Text!(
     Format="LOCTEXT(\"You\\'ve succesfully renamed and set correct previous values for {Int} input values in {MSName} renamed {MSInput} input.\")"
     Int=Length(TargetArray=@n19) {is=title K2Node_CallArrayFunction}
     MSName=Get SelectedMSName()
     MSInput:Name=String To Name(InString=@n21)
    ) -> (Result)
    MessageType=Ok
    DefaultValue=Ok
    MessageCategory=Info
   )
   > n23 = Switch on EAppReturnType(Selection=@n22)
   ClearCachedArrays(execute=@n23.Ok)
   n24 = Find(
    TargetMap:Map of Meta Sound Source Object References to Meta Sound Literal Structures=@n18
    Key:"Meta Sound Source Object Reference (by ref)"=@n20
   ) -> (Value:Meta Sound Literal Structure)
   Branch(execute=@n20.LoopBody, Condition=@n24.ReturnValue)
   n25 = Get Text(self=Get NewName())
   > SetGraphInputDefault!(
    execute
    self:Meta Sound Builder Base Object Reference=Find Or Begin Building MetaSound Asset(self=Get MetaSoundEditorSubsystem(), MetaSound=@n20)
    InputName:Name=String To Name(InString="To String (Text)"(InText=@n25))
    Literal:"Meta Sound Literal Structure (by ref)"=@n24
    PageName:Name
   ) -> (Succeeded:Exec, Failed:Exec)
   > SaveLoadedAsset(AssetToSave=@n20, bOnlyIfIsDirty=false)
   Event Construct()
   > ClearCachedArrays()
   Event Destruct()
   > ClearCachedArrays()
  ]}
 FindReferencers {nodes [
   n1 = Inputs!() -> (execute, Object:Object Reference)
   n2 = Get Object Path String(Object=@n1)
   > n3 = Get Referencers!(
    execute
    self:Asset Registry Interface=GetAssetRegistry()
    PackageName:Name=String To Name(InString=Split(SourceString=@n2, InStr=., SearchDir=FromEnd))
    ReferenceOptions:"Asset Registry Dependency Options Structure (by ref)"=Make Asset Registry Dependency Options(bIncludeSoftPackageReferences=True)
   ) -> (then, OutReferencers:Array of Names, ReturnValue:bool) {is=title K2Node_CallFunction}
   > n4 = For Each Loop(Array:Array of Names=@n3) -> (Array Element:Name)
   > n5 = GetAssetsByPackageName!(
    execute
    self:Asset Registry Interface=GetAssetRegistry()
    PackageName:Name=@n4
    bIncludeOnlyOnDiskAssets:bool=false
    bSkipARFilteredAssets:bool=true
   ) -> (then, OutAssetData:Array of Asset Data Structures, ReturnValue:bool)
   > Branch(Condition="integer > integer"(A=Length(TargetArray:Array of Asset Data Structures=@n5), B=0) {is=title K2Node_CallFunction})
   n6 = "Get (a copy)"(Array:Array of Asset Data Structures=@n5) -> (Output:Asset Data Structure)
   > n7 = LoadAsset_Blocking(Asset=To Soft Object Reference(SoftObjectPath=ToSoftObjectPath!(InAssetData:"Asset Data Structure (by ref)"=@n6) -> (ReturnValue:Soft Object Path Structure)))
   > Outputs!(Iteration:Exec, Referencer:Object Reference=@n7, Completed:Exec=@n4.Completed)
  ]}
]
variables [
 ReferencedMS {type="array<MetaSoundSource>", default []}
 SelectedMSName {type=string, default=""}
 TMapReferences {default {}}
]
widgets {
 info {rootWidgetClass=&CanvasPanel, widgetCount=12}
 widgets [
  CanvasPanel_21 {widgetClassPath=&CanvasPanel, widget=CanvasPanel_21}
  VerticalBox_65 {
   widgetClassPath=&VerticalBox
   parent=CanvasPanel_21
   widget=VerticalBox_65
   slot {
    class=CanvasPanelSlot
    properties {layoutData {alignment=<0 0>, anchors {maximum=<0 0>, minimum=<0 0>}, offsets=<ltrb 0 0 361.2762 144.1206>}}
   }
  }
  HorizontalBox_131 {widgetClassPath=&HorizontalBox, parent=VerticalBox_65, widget=HorizontalBox_131, slot {class=VerticalBoxSlot}}
  UEFN_TextBlock_C_2 {
   widgetClassPath=&UEFN_TextBlock
   parent=HorizontalBox_131
   widget=UEFN_TextBlock_C_2
   properties {
    text=Old Input Name
    toolTipText=Metasound preset unque name to added to prefixes and suffixes that are auto added.
    font {
     bForceMonospaced=false
     bMaterialIsStencil=false
     fontMaterial=null
     fontObject=&Roboto
     letterSpacing=0
     monospacedWidth=1
     size=10.5
     skewAmount=0
     typefaceFontName=Regular
     outlineSettings {
      bApplyOutlineToDropShadows=false
      bMiteredCorners=false
      bSeparateFillAlpha=false
      outlineMaterial=null
      outlineSize=0
      outlineColor=<rgba 0 0 0 1>
     }
    }
    shadowColorAndOpacity=<rgba 0 0 0 1>
   }
   slot {class=HorizontalBoxSlot, properties {padding=<ltrb 5 5 5 5>}}
  }
  OldName {
   widgetClassPath=&EditorUtilityEditableTextBox
   bIsVariable=true
   parent=HorizontalBox_131
   widget=OldName
   properties {toolTipText=MetaSound Old Input Name}
   slot {class=HorizontalBoxSlot, properties {padding=<ltrb 11 5 5 5>, size {sizeRule=Fill, value=1}}}
  }
  HorizontalBox {widgetClassPath=&HorizontalBox, parent=VerticalBox_65, widget=HorizontalBox, slot {class=VerticalBoxSlot}}
  UEFN_TextBlock_C {
   widgetClassPath=&UEFN_TextBlock
   parent=HorizontalBox
   widget=UEFN_TextBlock_C
   properties {
    text=New Input Name
    toolTipText=Metasound preset unque name to added to prefixes and suffixes that are auto added.
    font {
     bForceMonospaced=false
     bMaterialIsStencil=false
     fontMaterial=null
     fontObject=&Roboto
     letterSpacing=0
     monospacedWidth=1
     size=10.5
     skewAmount=0
     typefaceFontName=Regular
     outlineSettings {
      bApplyOutlineToDropShadows=false
      bMiteredCorners=false
      bSeparateFillAlpha=false
      outlineMaterial=null
      outlineSize=0
      outlineColor=<rgba 0 0 0 1>
     }
    }
    shadowColorAndOpacity=<rgba 0 0 0 1>
   }
   slot {class=HorizontalBoxSlot, properties {padding=<ltrb 5 5 5 5>}}
  }
  NewName {
   widgetClassPath=&EditorUtilityEditableTextBox
   bIsVariable=true
   parent=HorizontalBox
   widget=NewName
   properties {toolTipText=MetaSound New Input Name}
   slot {class=HorizontalBoxSlot, properties {padding=<ltrb 5 5 5 5>, size {sizeRule=Fill, value=1}}}
  }
  CacheValues {
   widgetClassPath=&EditorUtilityButton
   parent=VerticalBox_65
   widget=CacheValues
   properties {toolTipText=Stores current input values in memory}
   slot {class=VerticalBoxSlot, properties {padding=<ltrb 2 2 2 2>}}
  }
  UEFN_TextBlock_C_39 {
   widgetClassPath=&UEFN_TextBlock
   parent=CacheValues
   widget=UEFN_TextBlock_C_39
   properties {
    justification=Center
    text=Cache Values
    font {
     bForceMonospaced=false
     bMaterialIsStencil=false
     fontMaterial=null
     fontObject=&BurbankSmall
     letterSpacing=0
     monospacedWidth=1
     size=9
     skewAmount=0
     typefaceFontName=Black
     outlineSettings {
      bApplyOutlineToDropShadows=false
      bMiteredCorners=false
      bSeparateFillAlpha=false
      outlineMaterial=null
      outlineSize=0
      outlineColor=<rgba 0 0 0 1>
     }
    }
    shadowColorAndOpacity=<rgba 0 0 0 1>
    shadowOffset=<0.5 0.5>
   }
   slot {class=ButtonSlot}
  }
  SafeRename {
   widgetClassPath=&EditorUtilityButton
   parent=VerticalBox_65
   widget=SafeRename
   properties {toolTipText=Assigns cached values to renamed input}
   slot {class=VerticalBoxSlot, properties {padding=<ltrb 2 2 2 2>}}
  }
  UEFN_TextBlock_C_1 {
   widgetClassPath=&UEFN_TextBlock
   parent=SafeRename
   widget=UEFN_TextBlock_C_1
   properties {
    justification=Center
    text=Safe Rename
    font {
     bForceMonospaced=false
     bMaterialIsStencil=false
     fontMaterial=null
     fontObject=&BurbankSmall
     letterSpacing=0
     monospacedWidth=1
     size=9
     skewAmount=0
     typefaceFontName=Black
     outlineSettings {
      bApplyOutlineToDropShadows=false
      bMiteredCorners=false
      bSeparateFillAlpha=false
      outlineMaterial=null
      outlineSize=0
      outlineColor=<rgba 0 0 0 1>
     }
    }
    shadowColorAndOpacity=<rgba 0 0 0 1>
    shadowOffset=<0.5 0.5>
   }
   slot {class=ButtonSlot}
  }
 ]
}
