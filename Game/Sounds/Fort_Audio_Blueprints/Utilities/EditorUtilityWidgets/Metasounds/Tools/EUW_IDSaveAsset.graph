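# Editor utility widget for MetaSound audio tools that provides a "Save Asset" toggle and an event to save a given asset.
# PreConstruct points a Single Property View at the widget itself, exposing its SaveAsset boolean.
# SaveIdAsset transacts the passed object and, when the toggle is on, saves it even if not dirty, printing its name on success.
# An OnSourcePropertyChanged dispatcher is declared but never called within this widget.
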EditorUtilityWidgetBlueprint /Game/Sounds/Fort_Audio_Blueprints/Utilities/EditorUtilityWidgets/Metasounds/Tools/EUW_IDSaveAsset
graphs [
 EventGraph {nodes [
   n1 = SaveIdAsset!() -> (OutputDelegate, then, AssetToSave:Object Reference) {is=title K2Node_CustomEvent}
   > TransactObject(Object=@n1.AssetToSave)
   > Branch(Condition=Get SaveAsset() -> (SaveAsset:bool))
   > n2 = SaveLoadedAsset(AssetToSave=@n1.AssetToSave, bOnlyIfIsDirty=false)
   > Branch(Condition=@n2)
   > PrintString(InString=Append(A=GetDisplayName(Object=@n1.AssetToSave), B=" SAVED!"))
   Event PreConstruct()
   > SetObject(self=Get SaveAssetView() -> (SaveAssetView:Single Property View Object Reference), NewObject=Self-Reference())
  ]}
 OnSourcePropertyChanged {nodes [OnSourcePropertyChanged() {is=title K2Node_FunctionEntry}]}
]
variables [
 OnSourcePropertyChanged {default=()}
 SaveAsset {type=boolean}
]
widgets {
 info {rootWidgetClass=&CanvasPanel, widgetCount=4}
 widgets [
  CanvasPanel_0 {widgetClassPath=&CanvasPanel, widget=CanvasPanel_0}
  GridPanel_0 {
   widgetClassPath=&GridPanel
   parent=CanvasPanel_0
   widget=GridPanel_0
   slot {
    class=CanvasPanelSlot
    properties {bAutoSize=true, layoutData {alignment=<0 0>, anchors {maximum=<0 0>, minimum=<0 0>}, offsets=<ltrb 0 0 500 500>}}
   }
  }
  SizeBox_0 {
   widgetClassPath=&SizeBox
   parent=GridPanel_0
   widget=SizeBox_0
   properties {bOverride_WidthOverride=true, widthOverride=400}
   slot {class=GridSlot}
  }
  SaveAssetView {
   widgetClassPath=&SinglePropertyView
   parent=GridPanel_0
   widget=SaveAssetView
   properties {nameOverride=Save Asset, propertyName=SaveAsset}
   slot {class=GridSlot, properties {row=1}}
  }
 ]
}
