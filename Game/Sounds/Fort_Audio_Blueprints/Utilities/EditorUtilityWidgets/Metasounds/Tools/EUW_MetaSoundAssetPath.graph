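# Editor utility widget for MetaSound tools that collects a destination path and asset name for preset creation.
# Buttons and text edits broadcast dispatchers, leaving Make Preset and Get Assets Path logic to listeners, with SetAssetPath deriving the path from an object.
# Get Current Path copies the Content Browser path, and Make Folder creates a named subfolder there and updates the path.
# Import Files targets the entered path, falling back to the browser's.
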EditorUtilityWidgetBlueprint /Game/Sounds/Fort_Audio_Blueprints/Utilities/EditorUtilityWidgets/Metasounds/Tools/EUW_MetaSoundAssetPath
graphs [
 EventGraph {nodes [
   "On Pressed (CurrentPathButton)"() {is=title K2Node_ComponentBoundEvent}
   > Call Event on Currrent Path Button!(execute:Exec, self:EUW Meta Sound Asset Path Object Reference) -> (then:Exec) {is=title K2Node_CallDelegate}
   > Call On Inputs Text Updated!(execute:Exec, self:EUW Meta Sound Asset Path Object Reference) -> (then:Exec) {is=title K2Node_CallDelegate}
   n1 = GetCurrentContentBrowserPath!() -> (OutPath:String, ReturnValue:bool)
   > Set ValidFilePath(ValidFilePath:bool=@n1.ReturnValue) -> (Output_Get:bool)
   > "SetText (Text Box)"!(
    execute
    self:Text Box Object Reference=Get PathInput() -> (PathInput:Editor Utility Editable Text Box Object Reference)
    InText:Text="To Text (String)"(InString=@n1)
   ) -> (then)
   "On Pressed (AssetsPathButton)"() {is=title K2Node_ComponentBoundEvent}
   > Call Event on Asset Path Button!(execute:Exec, self:EUW Meta Sound Asset Path Object Reference) -> (then:Exec) {is=title K2Node_CallDelegate}
   > Call On Inputs Text Updated()
   "On Pressed (MakePresetButton)"() {is=title K2Node_ComponentBoundEvent}
   > Call Event on Make Preset Button!(execute:Exec, self:EUW Meta Sound Asset Path Object Reference) -> (then:Exec) {is=title K2Node_CallDelegate}
   "On Pressed (MakeFolderButton)"() {is=title K2Node_ComponentBoundEvent}
   > Call Event on Make Folder Button!(execute:Exec, self:EUW Meta Sound Asset Path Object Reference) -> (then:Exec) {is=title K2Node_CallDelegate}
   n2 = GetCurrentContentBrowserPath()
   > Set ValidFilePath(ValidFilePath=@n2.ReturnValue)
   n3 = GetAssetName!() -> (AssetText:Text, IsEmpty:bool)
   > n4 = Branch(Condition=@n3.IsEmpty)
   n5 = Append!(A="To String (Text)"(InText="To Text (String)"(InString=@n2)), B=/, C:String="To String (Text)"(InText=@n3)) -> (ReturnValue)
   MakeDirectory!(
    execute=@n4.else
    self:Editor Asset Subsystem Object Reference=Get EditorAssetSubsystem() -> (ReturnValue:Editor Asset Subsystem Object Reference) {is=title K2Node_GetEditorSubsystem}
    DirectoryPath:String=@n5
   ) -> (then, ReturnValue:bool)
   > "SetText (Text Box)"(self=Get PathInput(), InText="To Text (String)"(InString=@n5))
   > PrintString(InString="Asset Save  Path has been update to new Folder")
   "On Pressed (ImportFilesButton)"() {is=title K2Node_ComponentBoundEvent}
   > Call Event on Import Files Button!(execute:Exec, self:EUW Meta Sound Asset Path Object Reference) -> (then:Exec) {is=title K2Node_CallDelegate}
   > n6 = GetAssetTools()
   n7 = Get Text(self=Get PathInput())
   n8 = "To Text (String)"(InString=GetCurrentContentBrowserPath())
   > ImportAssetsWithDialog!(
    execute
    self:Asset Tools Interface=@n6
    DestinationPath:String=SelectString(A="To String (Text)"(InText=@n8), B="To String (Text)"(InText=@n7), bPickA=TextIsEmpty(InText=@n7))
   ) -> (then, ReturnValue:Array of Object References)
   "On Text Committed (PathInput)"!() -> (OutputDelegate, then, Text:"Text (by ref)", CommitMethod:ETextCommit Enum=Default) {is=title K2Node_ComponentBoundEvent}
   > Call On Inputs Text Updated()
   "On Text Changed (AssetNameInput)"!() -> (OutputDelegate, then, Text:"Text (by ref)") {is=title K2Node_ComponentBoundEvent}
   > Call On Inputs Text Updated()
   n9 = OnAssetNameTextChanged!() -> (OutputDelegate, then, Text:"Text (by ref)") {is=title K2Node_CustomEvent}
   Call On Inputs Text Updated(execute=[@n9.then, @n10.then])
   n10 = OnPathInputTextChanged!() -> (OutputDelegate, then, Text:"Text (by ref)") {is=title K2Node_CustomEvent}
   Event Construct()
   > Bind Event to On Text Changed(
    self:Text Box Object Reference=Get AssetNameInput() -> (AssetNameInput:Editor Utility Editable Text Box Object Reference)
    Delegate=Create Event()
   ) {is=title K2Node_AddDelegate}
   > Assign On Text Changed!(execute:Exec, self:Text Box Object Reference=Get PathInput(), Delegate:"Delegate (by ref)"=Create Event()) -> (then:Exec) {is=title K2Node_AssignDelegate}
   Reroute Node(InputPin:Wildcard) -> (OutputPin:Wildcard)
  ]}
 EventOnAssetPathButton {nodes [EventOnAssetPathButton() {is=title K2Node_FunctionEntry}]}
 EventOnCurrrentPathButton {nodes [EventOnCurrrentPathButton() {is=title K2Node_FunctionEntry}]}
 EventOnImportFilesButton {nodes [EventOnImportFilesButton() {is=title K2Node_FunctionEntry}]}
 EventOnMakeFolderButton {nodes [EventOnMakeFolderButton() {is=title K2Node_FunctionEntry}]}
 EventOnMakePresetButton {nodes [EventOnMakePresetButton() {is=title K2Node_FunctionEntry}]}
 GetAssetName {nodes [
   GetAssetName() {is=title K2Node_FunctionEntry}
   n1 = Get Text(self=Get AssetNameInput())
   n2 = "Equal Exactly (String)"(A="To String (Text)"(InText=@n1))
   > n3 = Branch(Condition=@n2)
   > n4 = Show Message Dialog(
    Title="To Text (String)"(InString=Asset Name Missing)
    Message="To Text (String)"(InString=No asset name input selected. Please input an asset name to use this feature.)
    MessageType=Ok
    DefaultValue=No
    MessageCategory=Info
   )
   Return Node!(execute=[@n4.then, @n3.else], AssetText:Text=@n1, IsEmpty:bool=@n2)
  ]}
 OnInputsTextUpdated {nodes [OnInputsTextUpdated() {is=title K2Node_FunctionEntry}]}
 SetAssetPath {nodes [
   n1 = SetAssetPath!() -> (then, Object:Object Reference) {is=title K2Node_FunctionEntry}
   n2 = Split(SourceString=Get Object Path String(Object=@n1), InStr=/, SearchDir=FromEnd)
   > Set ValidFilePath(ValidFilePath=@n2.ReturnValue)
   > "SetText (Text Box)"(self=Get PathInput(), InText="To Text (String)"(InString=@n2))
  ]}
]
variables [
 EventOnAssetPathButton {default=()}
 EventOnCurrrentPathButton {default=()}
 EventOnImportFilesButton {default=()}
 EventOnMakeFolderButton {default=()}
 EventOnMakePresetButton {default=()}
 OnInputsTextUpdated {default=()}
 ValidFilePath {type=boolean}
]
widgets {
 info {rootWidgetClass=&CanvasPanel, widgetCount=27}
 widgets [
  CanvasPanel_0 {widgetClassPath=&CanvasPanel, widget=CanvasPanel_0}
  GridPanel_143 {
   widgetClassPath=&GridPanel
   parent=CanvasPanel_0
   widget=GridPanel_143
   slot {
    class=CanvasPanelSlot
    properties {bAutoSize=true, layoutData {alignment=<0 0>, anchors {maximum=<0 0>, minimum=<0 0>}, offsets=<ltrb 0 0 300 30>}}
   }
  }
  SizeBox_0 {
   widgetClassPath=&SizeBox
   parent=GridPanel_143
   widget=SizeBox_0
   properties {bOverride_HeightOverride=true, bOverride_WidthOverride=true, heightOverride=50, widthOverride=200}
   slot {class=GridSlot, properties {columnSpan=2, horizontalAlignment=HAlign_Center, row=2, verticalAlignment=VAlign_Center, padding=<ltrb 0 5 0 5>}}
  }
  MakePresetButton {
   widgetClassPath=&EditorUtilityButton
   parent=SizeBox_0
   widget=MakePresetButton
   properties {toolTipText="Creates Meta Sound from all selected waves to make cosmetic preset(s)."}
   slot {class=SizeBoxSlot}
  }
  UEFN_TextBlock_C_39 {
   widgetClassPath=&UEFN_TextBlock
   parent=MakePresetButton
   widget=UEFN_TextBlock_C_39
   properties {
    justification=Center
    text=Make Preset
    font {
     bForceMonospaced=false
     bMaterialIsStencil=false
     fontMaterial=null
     fontObject=&BurbankSmall
     letterSpacing=0
     monospacedWidth=1
     size=9
     skewAmount=0
     typefaceFontName=Black
     outlineSettings {
      bApplyOutlineToDropShadows=false
      bMiteredCorners=false
      bSeparateFillAlpha=false
      outlineMaterial=null
      outlineSize=0
      outlineColor=<rgba 0 0 0 1>
     }
    }
    shadowColorAndOpacity=<rgba 0 0 0 1>
   }
   slot {class=ButtonSlot}
  }
  VerticalBox_0 {widgetClassPath=&VerticalBox, parent=GridPanel_143, widget=VerticalBox_0, slot {class=GridSlot, properties {column=1}}}
  SizeBox_3 {
   widgetClassPath=&SizeBox
   parent=VerticalBox_0
   widget=SizeBox_3
   properties {bOverride_HeightOverride=true, heightOverride=40}
   slot {class=VerticalBoxSlot, properties {padding=<ltrb 2 2 2 2>}}
  }
  CurrentPathButton {
   widgetClassPath=&EditorUtilityButton
   parent=SizeBox_3
   widget=CurrentPathButton
   properties {toolTipText=Set current editor path as the preset build destination.}
   slot {class=SizeBoxSlot}
  }
  UEFN_TextBlock_C_1 {
   widgetClassPath=&UEFN_TextBlock
   parent=CurrentPathButton
   widget=UEFN_TextBlock_C_1
   properties {
    justification=Center
    text=Get Current Path
    font {
     bForceMonospaced=false
     bMaterialIsStencil=false
     fontMaterial=null
     fontObject=&Roboto
     letterSpacing=0
     monospacedWidth=1
     size=9
     skewAmount=0
     typefaceFontName=Regular
     outlineSettings {
      bApplyOutlineToDropShadows=false
      bMiteredCorners=false
      bSeparateFillAlpha=false
      outlineMaterial=null
      outlineSize=0
      outlineColor=<rgba 0 0 0 1>
     }
    }
    shadowColorAndOpacity=<rgba 0 0 0 1>
   }
   slot {class=ButtonSlot}
  }
  SizeBox_2 {
   widgetClassPath=&SizeBox
   parent=VerticalBox_0
   widget=SizeBox_2
   properties {bOverride_HeightOverride=true, heightOverride=40}
   slot {class=VerticalBoxSlot, properties {padding=<ltrb 2 2 2 2>}}
  }
  AssetsPathButton {
   widgetClassPath=&EditorUtilityButton
   parent=SizeBox_2
   widget=AssetsPathButton
   properties {toolTipText=Set path to selected assets as the preset build destination.}
   slot {class=SizeBoxSlot}
  }
  UEFN_TextBlock_C_3 {
   widgetClassPath=&UEFN_TextBlock
   parent=AssetsPathButton
   widget=UEFN_TextBlock_C_3
   properties {
    justification=Center
    text=Get Assets Path
    font {
     bForceMonospaced=false
     bMaterialIsStencil=false
     fontMaterial=null
     fontObject=&Roboto
     letterSpacing=0
     monospacedWidth=1
     size=9
     skewAmount=0
     typefaceFontName=Regular
     outlineSettings {
      bApplyOutlineToDropShadows=false
      bMiteredCorners=false
      bSeparateFillAlpha=false
      outlineMaterial=null
      outlineSize=0
      outlineColor=<rgba 0 0 0 1>
     }
    }
    shadowColorAndOpacity=<rgba 0 0 0 1>
   }
   slot {class=ButtonSlot}
  }
  SizeBox_1 {
   widgetClassPath=&SizeBox
   parent=GridPanel_143
   widget=SizeBox_1
   properties {bOverride_WidthOverride=true, widthOverride=350}
   slot {class=GridSlot}
  }
  VerticalBox_213 {
   widgetClassPath=&VerticalBox
   parent=SizeBox_1
   widget=VerticalBox_213
   slot {class=SizeBoxSlot, properties {verticalAlignment=VAlign_Center}}
  }
  HorizontalBox_79 {
   widgetClassPath=&HorizontalBox
   parent=VerticalBox_213
   widget=HorizontalBox_79
   slot {class=VerticalBoxSlot, properties {verticalAlignment=VAlign_Center}}
  }
  UEFN_TextBlock_C {
   widgetClassPath=&UEFN_TextBlock
   parent=HorizontalBox_79
   widget=UEFN_TextBlock_C
   properties {
    text=Path
    toolTipText=Path that meta sound asset will be added to.
    font {
     bForceMonospaced=false
     bMaterialIsStencil=false
     fontMaterial=null
     fontObject=&Roboto
     letterSpacing=0
     monospacedWidth=1
     size=10.5
     skewAmount=0
     typefaceFontName=Regular
     outlineSettings {
      bApplyOutlineToDropShadows=false
      bMiteredCorners=false
      bSeparateFillAlpha=false
      outlineMaterial=null
      outlineSize=0
      outlineColor=<rgba 0 0 0 1>
     }
    }
    shadowColorAndOpacity=<rgba 0 0 0 1>
   }
   slot {class=HorizontalBoxSlot, properties {padding=<ltrb 5 5 5 5>}}
  }
  PathInput {
   widgetClassPath=&EditorUtilityEditableTextBox
   bIsVariable=true
   parent=HorizontalBox_79
   widget=PathInput
   properties {toolTipText=Path that meta sound asset will be added to.}
   slot {class=HorizontalBoxSlot, properties {padding=<ltrb 5 5 5 5>, size {sizeRule=Fill, value=1}}}
  }
  HorizontalBox_131 {widgetClassPath=&HorizontalBox, parent=VerticalBox_213, widget=HorizontalBox_131, slot {class=VerticalBoxSlot}}
  UEFN_TextBlock_C_2 {
   widgetClassPath=&UEFN_TextBlock
   parent=HorizontalBox_131
   widget=UEFN_TextBlock_C_2
   properties {
    text=Asset Name
    toolTipText=Metasound preset unque name to added to prefixes and suffixes that are auto added.
    font {
     bForceMonospaced=false
     bMaterialIsStencil=false
     fontMaterial=null
     fontObject=&Roboto
     letterSpacing=0
     monospacedWidth=1
     size=10.5
     skewAmount=0
     typefaceFontName=Regular
     outlineSettings {
      bApplyOutlineToDropShadows=false
      bMiteredCorners=false
      bSeparateFillAlpha=false
      outlineMaterial=null
      outlineSize=0
      outlineColor=<rgba 0 0 0 1>
     }
    }
    shadowColorAndOpacity=<rgba 0 0 0 1>
   }
   slot {class=HorizontalBoxSlot, properties {padding=<ltrb 5 5 5 5>}}
  }
  AssetNameInput {
   widgetClassPath=&EditorUtilityEditableTextBox
   bIsVariable=true
   parent=HorizontalBox_131
   widget=AssetNameInput
   properties {toolTipText=Metasound preset unque name to added to prefixes and suffixes that are auto added.}
   slot {class=HorizontalBoxSlot, properties {padding=<ltrb 5 5 5 5>, size {sizeRule=Fill, value=1}}}
  }
  HorizontalBox_0 {
   widgetClassPath=&HorizontalBox
   parent=GridPanel_143
   widget=HorizontalBox_0
   slot {class=GridSlot, properties {row=1, verticalAlignment=VAlign_Center, padding=<ltrb 5 2 2 2>}}
  }
  SizeBox_5 {
   widgetClassPath=&SizeBox
   parent=HorizontalBox_0
   widget=SizeBox_5
   properties {bOverride_HeightOverride=true, heightOverride=35}
   slot {class=HorizontalBoxSlot, properties {padding=<ltrb 2 2 2 2>}}
  }
  MakeFolderButton {
   widgetClassPath=&EditorUtilityButton
   parent=SizeBox_5
   widget=MakeFolderButton
   properties {toolTipText=Set path to selected assets as the preset build destination.}
   slot {class=SizeBoxSlot}
  }
  UEFN_TextBlock_C_4 {
   widgetClassPath=&UEFN_TextBlock
   parent=MakeFolderButton
   widget=UEFN_TextBlock_C_4
   properties {
    justification=Center
    text=Make Folder
    font {
     bForceMonospaced=false
     bMaterialIsStencil=false
     fontMaterial=null
     fontObject=&Roboto
     letterSpacing=0
     monospacedWidth=1
     size=9
     skewAmount=0
     typefaceFontName=Regular
     outlineSettings {
      bApplyOutlineToDropShadows=false
      bMiteredCorners=false
      bSeparateFillAlpha=false
      outlineMaterial=null
      outlineSize=0
      outlineColor=<rgba 0 0 0 1>
     }
    }
    shadowColorAndOpacity=<rgba 0 0 0 1>
   }
   slot {class=ButtonSlot}
  }
  SizeBox_4 {
   widgetClassPath=&SizeBox
   parent=HorizontalBox_0
   widget=SizeBox_4
   properties {bOverride_HeightOverride=true, heightOverride=35}
   slot {class=HorizontalBoxSlot, properties {padding=<ltrb 2 2 2 2>}}
  }
  ImportFilesButton {
   widgetClassPath=&EditorUtilityButton
   parent=SizeBox_4
   widget=ImportFilesButton
   properties {toolTipText=Set path to selected assets as the preset build destination.}
   slot {class=SizeBoxSlot}
  }
  UEFN_TextBlock_C_5 {
   widgetClassPath=&UEFN_TextBlock
   parent=ImportFilesButton
   widget=UEFN_TextBlock_C_5
   properties {
    justification=Center
    text=Import Files
    font {
     bForceMonospaced=false
     bMaterialIsStencil=false
     fontMaterial=null
     fontObject=&Roboto
     letterSpacing=0
     monospacedWidth=1
     size=9
     skewAmount=0
     typefaceFontName=Regular
     outlineSettings {
      bApplyOutlineToDropShadows=false
      bMiteredCorners=false
      bSeparateFillAlpha=false
      outlineMaterial=null
      outlineSize=0
      outlineColor=<rgba 0 0 0 1>
     }
    }
    shadowColorAndOpacity=<rgba 0 0 0 1>
   }
   slot {class=ButtonSlot}
  }
 ]
}
